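# Physics asset for two dynamic strap chains (A and B) on the CoyoTear medium female player body.
# Each chain runs three capsule segments from a small master capsule constrained to spine_05, with limited swing and very tight twist between links.
# Larger capsules on spine_04, spine_05 and the right upper-arm twist bone give the torso and arm volume.
# Constraints joining the masters and these body capsules leave swing and twist free.
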PhysicsAsset /Game/Characters/Player/Female/Medium/Bodies/F_MED_CoyoTear/Meshes/F_MED_CoyoTear_Dyn_Straps
bodies {
 dyn_straps_A_01 [{shapeType=Capsule, length=6, radius=1, shapeName=dyn_straps_A_01_capsule, center=<4 0 0>, rotation=<rot 90 0 0>}]
 dyn_straps_A_02 [{shapeType=Capsule, length=5, radius=1, shapeName=dyn_straps_A_02_capsule, center=<4 0 0>, rotation=<rot 90 0 0>}]
 dyn_straps_A_03 [{shapeType=Capsule, length=5, radius=1, shapeName=dyn_straps_A_03_capsule, center=<4 0 0>, rotation=<rot 90 0 0>}]
 dyn_straps_A_MSTR [{
   shapeType=Capsule
   length=1.2017
   radius=0.6358
   shapeName=dyn_straps_A_MSTR_capsule
   center=<0.6809 -0.1104 -0.1179>
   rotation=<rot 0 -105.985 -72.4302>
  }]
 dyn_straps_B_01 [{shapeType=Capsule, length=6, radius=1, shapeName=dyn_straps_B_01_capsule, center=<4 0 0>, rotation=<rot 90 0 0>}]
 dyn_straps_B_02 [{shapeType=Capsule, length=5, radius=1, shapeName=dyn_straps_B_02_capsule, center=<4 0 0>, rotation=<rot 90 0 0>}]
 dyn_straps_B_03 [{shapeType=Capsule, length=5, radius=1, shapeName=dyn_straps_B_03_capsule, center=<4 0 0>, rotation=<rot 90 0 0>}]
 dyn_straps_B_MSTR [{
   shapeType=Capsule
   length=1.2022
   radius=0.6353
   shapeName=dyn_straps_B_MSTR_capsule
   center=<0.6759 -0.1114 -0.1243>
   rotation=<rot 0 -106.5208 -71.8412>
  }]
 spine_04 [{shapeType=Capsule, length=6.0096, radius=10.4925, shapeName=spine_04_capsule, center=<0.8376 1.2133 0>}]
 spine_05 [{shapeType=Capsule, length=4.5096, radius=11.9925, shapeName=spine_04_capsule, center=<-4.6221 -0.72 0>}]
 upperarm_twist_01_r [{shapeType=Capsule, length=16.5096, radius=3.4925, shapeName=spine_04_capsule, center=<-2.6073 0.2207 0.5658>, rotation=<rot 90 180 -176.863>}]
}
constraints [
 {bone1Name=dyn_straps_A_01, bone2Name=dyn_straps_A_MSTR, swing1LimitDegrees=60, swing2LimitDegrees=45, twistLimitDegrees=2}
 {bone1Name=dyn_straps_A_02, bone2Name=dyn_straps_A_01, swing1LimitDegrees=45, swing2LimitDegrees=45, twistLimitDegrees=2}
 {bone1Name=dyn_straps_A_03, bone2Name=dyn_straps_A_02, swing1LimitDegrees=45, swing2LimitDegrees=45, twistLimitDegrees=2}
 {bone1Name=dyn_straps_B_01, bone2Name=dyn_straps_B_MSTR, swing1LimitDegrees=60, swing2LimitDegrees=45, twistLimitDegrees=2}
 {bone1Name=dyn_straps_B_02, bone2Name=dyn_straps_B_01, swing1LimitDegrees=45, swing2LimitDegrees=45, twistLimitDegrees=2}
 {bone1Name=dyn_straps_B_03, bone2Name=dyn_straps_B_02, swing1LimitDegrees=45, swing2LimitDegrees=45, twistLimitDegrees=2}
 {
  bone1Name=dyn_straps_A_MSTR
  bone2Name=spine_05
  swing1LimitDegrees=45
  swing1Motion=Free
  swing2LimitDegrees=45
  swing2Motion=Free
  twistLimitDegrees=45
  twistMotion=Free
 }
 {
  bone1Name=dyn_straps_B_MSTR
  bone2Name=spine_05
  swing1LimitDegrees=45
  swing1Motion=Free
  swing2LimitDegrees=45
  swing2Motion=Free
  twistLimitDegrees=45
  twistMotion=Free
 }
 {
  bone1Name=spine_05
  bone2Name=spine_04
  swing1LimitDegrees=45
  swing1Motion=Free
  swing2LimitDegrees=45
  swing2Motion=Free
  twistLimitDegrees=45
  twistMotion=Free
 }
 {
  bone1Name=upperarm_twist_01_r
  bone2Name=spine_05
  swing1LimitDegrees=45
  swing1Motion=Free
  swing2LimitDegrees=45
  swing2Motion=Free
  twistLimitDegrees=45
  twistMotion=Free
 }
]
