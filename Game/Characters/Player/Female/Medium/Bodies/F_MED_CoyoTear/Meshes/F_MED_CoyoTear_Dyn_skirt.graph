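# Physics asset for the CoyoTear medium female player body's dynamic skirt, built from left, right, back-left, back-right and front-left coat bone chains.
# Each chain hangs from a tiny MSTR body through swing- and twist-limited joints, with large offset capsules along it.
# Pelvis and leg capsules plus a huge sphere below serve as body and ground colliders.
# Angular-free constraints cross-link side and back chains at the fourth segment on each side.
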PhysicsAsset /Game/Characters/Player/Female/Medium/Bodies/F_MED_CoyoTear/Meshes/F_MED_CoyoTear_Dyn_skirt
bodies {
 calf_l [{
   shapeType=Capsule
   length=34.7349
   radius=9.2984
   shapeName=pelvis_capsule
   center=<-22.0673 -2.9305 0.0431>
   rotation=<rot 86.3784 179.9315 178.6153>
  }]
 calf_r [{
   shapeType=Capsule
   length=27.7349
   radius=11.7984
   shapeName=pelvis_capsule
   center=<18.7818 2.161 -1.4781>
   rotation=<rot 89.5502 -160.9775 21.7228>
  }]
 dyn_coat_bk_l_01 [{shapeType=Capsule, length=6.4264, radius=18.375, shapeName=dyn_coat_l_01_capsule, center=<5 0 18>, rotation=<rot 0 0 90>}]
 dyn_coat_bk_l_02 [{shapeType=Capsule, length=12.5061, radius=18.375, shapeName=dyn_coat_l_01_capsule, center=<5 0 18>, rotation=<rot 0 0 90>}]
 dyn_coat_bk_l_03 [{shapeType=Capsule, length=27.1311, radius=18.625, shapeName=dyn_coat_l_01_capsule, center=<5 0 18>, rotation=<rot -0.0001 0 95>}]
 dyn_coat_bk_l_04 [{shapeType=Capsule, length=27.2561, radius=19.75, shapeName=dyn_coat_l_01_capsule, center=<5 0 18>, rotation=<rot -0.0001 0 95>}]
 dyn_coat_bk_l_05 [{shapeType=Capsule, length=35.2561, radius=19.25, shapeName=dyn_coat_l_01_capsule, center=<5 0 18>, rotation=<rot -0.0001 0 95>}]
 dyn_coat_bk_l_MSTR [{shapeType=Capsule, length=0.505, radius=0.505, shapeName=dyn_coat_l_MSTR_capsule}]
 dyn_coat_bk_r_01 [{shapeType=Capsule, length=13.4264, radius=17.375, shapeName=dyn_coat_l_01_capsule, center=<5 0 18>, rotation=<rot -0.0001 0 75>}]
 dyn_coat_bk_r_02 [{shapeType=Capsule, length=19.5061, radius=17.375, shapeName=dyn_coat_l_01_capsule, center=<5 0 18>, rotation=<rot -0.0001 0 75>}]
 dyn_coat_bk_r_03 [{shapeType=Capsule, length=34.6311, radius=18.125, shapeName=dyn_coat_l_01_capsule, center=<6 0 18>, rotation=<rot -0.0004 0 75>}]
 dyn_coat_bk_r_04 [{shapeType=Capsule, length=34.2561, radius=19.75, shapeName=dyn_coat_l_01_capsule, center=<6 0 18>, rotation=<rot 0.8667 4.9244 75.0373>}]
 dyn_coat_bk_r_05 [{shapeType=Capsule, length=42.2561, radius=19.25, shapeName=dyn_coat_l_01_capsule, center=<6 0 18>, rotation=<rot 1.7273 9.8511 75.1488>}]
 dyn_coat_bk_r_MSTR [{shapeType=Capsule, length=0.505, radius=0.505, shapeName=dyn_coat_l_MSTR_capsule}]
 dyn_coat_ft_l_01 [{shapeType=Capsule, length=15.5, radius=15.75, shapeName=dyn_coat_l_01_capsule, center=<5 -10 15>, rotation=<rot 0.0003 0.0002 79.9999>}]
 dyn_coat_ft_l_02 [{shapeType=Capsule, length=13.5, radius=17.5, shapeName=dyn_coat_l_01_capsule, center=<5 -10 15>, rotation=<rot 0.0003 0.0002 79.9999>}]
 dyn_coat_ft_l_03 [{shapeType=Capsule, length=13.5, radius=18.75, shapeName=dyn_coat_l_01_capsule, center=<10.3356 -10 15>, rotation=<rot 0.0003 0.0002 79.9999>}]
 dyn_coat_ft_l_04 [{shapeType=Capsule, length=14.5, radius=19.25, shapeName=dyn_coat_l_01_capsule, center=<10.3356 -10 15>, rotation=<rot 0.0003 0.0002 79.9999>}]
 dyn_coat_ft_l_MSTR [{shapeType=Capsule, length=0.505, radius=0.505, shapeName=dyn_coat_l_MSTR_capsule}]
 dyn_coat_l_01 [{shapeType=Capsule, length=19.5, radius=17.25, shapeName=dyn_coat_l_01_capsule, center=<5 0 17>, rotation=<rot 0 -0.0003 81.7049>}]
 dyn_coat_l_02 [{shapeType=Capsule, length=19, radius=17.5, shapeName=dyn_coat_l_01_capsule, center=<5 0 17>, rotation=<rot 0 -0.0003 81.7049>}]
 dyn_coat_l_03 [{shapeType=Capsule, length=27.5, radius=18.25, shapeName=dyn_coat_l_01_capsule, center=<5 0 17>, rotation=<rot 0 -0.0003 81.7049>}]
 dyn_coat_l_04 [{shapeType=Capsule, length=27, radius=20.25, shapeName=dyn_coat_l_01_capsule, center=<5 0 17>, rotation=<rot 0 -0.0003 81.7049>}]
 dyn_coat_l_05 [{shapeType=Capsule, length=27, radius=20.25, shapeName=dyn_coat_l_01_capsule, center=<5 0 17>, rotation=<rot 0 -0.0003 81.7049>}]
 dyn_coat_l_MSTR [{shapeType=Capsule, length=0.505, radius=0.505, shapeName=dyn_coat_l_MSTR_capsule}]
 dyn_coat_r_01 [{shapeType=Capsule, length=17, radius=9.75, shapeName=dyn_coat_l_01_capsule, center=<7 0 12>, rotation=<rot 0 -0.0018 90.0002>}]
 dyn_coat_r_02 [{shapeType=Capsule, length=15, radius=14.5, shapeName=dyn_coat_l_01_capsule, center=<7 0 15>, rotation=<rot 0 -0.0009 95.0003>}]
 dyn_coat_r_03 [{shapeType=Capsule, length=23.5, radius=20.75, shapeName=dyn_coat_l_01_capsule, center=<10.3356 10 17>, rotation=<rot 0 -0.0007 120.0001>}]
 dyn_coat_r_04 [{shapeType=Capsule, length=28, radius=21.25, shapeName=dyn_coat_l_01_capsule, center=<10.3356 10 17>, rotation=<rot 0 -0.0006 120.0001>}]
 dyn_coat_r_05 [{shapeType=Capsule, length=27, radius=24.25, shapeName=dyn_coat_l_01_capsule, center=<10.3168 10 17>, rotation=<rot -3.2744 5.6861 119.8374>}]
 dyn_coat_r_MSTR [{shapeType=Capsule, length=0.505, radius=0.505, shapeName=dyn_coat_l_MSTR_capsule}]
 dyn_ground_collision [{shapeType=Sphere, radius=300, shapeName=root_sphere, center=<4 -25 -291.1913>}]
 foot_l [{shapeType=Capsule, length=9.6678, radius=13.2363, shapeName=foot_r_capsule, center=<-2.2209 5.395 0>, rotation=<rot 0 0 -90>}]
 foot_r [{shapeType=Capsule, length=12.1678, radius=12.7363, shapeName=foot_r_capsule, center=<3.0534 -5.395 0>, rotation=<rot 0 0 90>}]
 pelvis [{shapeType=Capsule, length=5.342, radius=13.146, shapeName=pelvis_capsule, center=<-4.8171 -0.6518 -0.0005>}]
 thigh_l [{
   shapeType=Capsule
   length=22.466
   radius=10.8021
   shapeName=pelvis_capsule
   center=<-24.848 -2.0695 1.2215>
   rotation=<rot 84.9228 -176.6941 179.2869>
  }]
 thigh_r [{
   shapeType=Capsule
   length=28.341
   radius=12.15
   shapeName=pelvis_capsule
   center=<21.7259 2.935 -1.8065>
   rotation=<rot 87.5479 179.9926 -2.575>
  }]
}
constraints [
 {bone1Name=dyn_coat_l_01, bone2Name=dyn_coat_l_MSTR, swing1LimitDegrees=10, swing2LimitDegrees=45, twistLimitDegrees=2}
 {bone1Name=dyn_coat_l_02, bone2Name=dyn_coat_l_01, swing1LimitDegrees=10, swing2LimitDegrees=45, twistLimitDegrees=2}
 {bone1Name=dyn_coat_l_03, bone2Name=dyn_coat_l_02, swing1LimitDegrees=10, swing2LimitDegrees=45, twistLimitDegrees=2}
 {bone1Name=dyn_coat_l_04, bone2Name=dyn_coat_l_03, swing1LimitDegrees=10, swing2LimitDegrees=45, twistLimitDegrees=2}
 {bone1Name=dyn_coat_r_01, bone2Name=dyn_coat_r_MSTR, swing1LimitDegrees=10, swing2LimitDegrees=60, twistLimitDegrees=2}
 {bone1Name=dyn_coat_r_02, bone2Name=dyn_coat_r_01, swing1LimitDegrees=10, swing2LimitDegrees=60, twistLimitDegrees=2}
 {bone1Name=dyn_coat_r_03, bone2Name=dyn_coat_r_02, swing1LimitDegrees=10, swing2LimitDegrees=45, twistLimitDegrees=2}
 {bone1Name=dyn_coat_r_04, bone2Name=dyn_coat_r_03, swing1LimitDegrees=10, swing2LimitDegrees=45, twistLimitDegrees=2}
 {bone1Name=dyn_coat_bk_l_01, bone2Name=dyn_coat_bk_l_MSTR, swing1LimitDegrees=10, swing2LimitDegrees=60, twistLimitDegrees=2}
 {bone1Name=dyn_coat_bk_l_02, bone2Name=dyn_coat_bk_l_01, swing1LimitDegrees=10, swing2LimitDegrees=60, twistLimitDegrees=2}
 {bone1Name=dyn_coat_bk_r_01, bone2Name=dyn_coat_bk_r_MSTR, swing1LimitDegrees=10, swing2LimitDegrees=60, twistLimitDegrees=2}
 {bone1Name=dyn_coat_bk_r_02, bone2Name=dyn_coat_bk_r_01, swing1LimitDegrees=10, swing2LimitDegrees=60, twistLimitDegrees=2}
 {
  bone1Name=dyn_coat_r_MSTR
  bone2Name=thigh_r
  swing1LimitDegrees=45
  swing1Motion=Free
  swing2LimitDegrees=45
  swing2Motion=Free
  twistLimitDegrees=45
  twistMotion=Free
 }
 {
  bone1Name=thigh_l
  bone2Name=pelvis
  swing1LimitDegrees=45
  swing1Motion=Free
  swing2LimitDegrees=45
  swing2Motion=Free
  twistLimitDegrees=45
  twistMotion=Free
 }
 {bone1Name=dyn_coat_bk_r_05, bone2Name=dyn_coat_bk_r_04, swing1LimitDegrees=10, swing2LimitDegrees=45, twistLimitDegrees=2}
 {bone1Name=dyn_coat_bk_l_05, bone2Name=dyn_coat_bk_l_04, swing1LimitDegrees=10, swing2LimitDegrees=45, twistLimitDegrees=2}
 {bone1Name=dyn_coat_bk_r_04, bone2Name=dyn_coat_bk_r_03, swing1LimitDegrees=10, swing2LimitDegrees=45, twistLimitDegrees=2}
 {bone1Name=dyn_coat_bk_l_04, bone2Name=dyn_coat_bk_l_03, swing1LimitDegrees=10, swing2LimitDegrees=45, twistLimitDegrees=2}
 {bone1Name=dyn_coat_bk_r_03, bone2Name=dyn_coat_bk_r_02, swing1LimitDegrees=10, swing2LimitDegrees=45, twistLimitDegrees=2}
 {bone1Name=dyn_coat_bk_l_03, bone2Name=dyn_coat_bk_l_02, swing1LimitDegrees=10, swing2LimitDegrees=45, twistLimitDegrees=2}
 {
  bone1Name=thigh_r
  bone2Name=pelvis
  swing1LimitDegrees=45
  swing1Motion=Free
  swing2LimitDegrees=45
  swing2Motion=Free
  twistLimitDegrees=45
  twistMotion=Free
 }
 {
  bone1Name=calf_r
  bone2Name=thigh_r
  swing1LimitDegrees=45
  swing1Motion=Free
  swing2LimitDegrees=45
  swing2Motion=Free
  twistLimitDegrees=45
  twistMotion=Free
 }
 {
  bone1Name=foot_r
  bone2Name=calf_r
  swing1LimitDegrees=45
  swing1Motion=Free
  swing2LimitDegrees=45
  swing2Motion=Free
  twistLimitDegrees=45
  twistMotion=Free
 }
 {
  bone1Name=calf_l
  bone2Name=thigh_l
  swing1LimitDegrees=45
  swing1Motion=Free
  swing2LimitDegrees=45
  swing2Motion=Free
  twistLimitDegrees=45
  twistMotion=Free
 }
 {
  bone1Name=foot_l
  bone2Name=calf_l
  swing1LimitDegrees=45
  swing1Motion=Free
  swing2LimitDegrees=45
  swing2Motion=Free
  twistLimitDegrees=45
  twistMotion=Free
 }
 {
  bone1Name=dyn_coat_ft_l_MSTR
  bone2Name=thigh_l
  swing1LimitDegrees=45
  swing1Motion=Free
  swing2LimitDegrees=45
  swing2Motion=Free
  twistLimitDegrees=45
  twistMotion=Free
 }
 {bone1Name=dyn_coat_ft_l_01, bone2Name=dyn_coat_ft_l_MSTR, swing1LimitDegrees=10, swing2LimitDegrees=45, twistLimitDegrees=2}
 {bone1Name=dyn_coat_ft_l_02, bone2Name=dyn_coat_ft_l_01, swing1LimitDegrees=5, swing2LimitDegrees=5, twistLimitDegrees=2}
 {bone1Name=dyn_coat_ft_l_03, bone2Name=dyn_coat_ft_l_02, swing1LimitDegrees=10, swing2LimitDegrees=45, twistLimitDegrees=2}
 {bone1Name=dyn_coat_ft_l_04, bone2Name=dyn_coat_ft_l_03, swing1LimitDegrees=10, swing2LimitDegrees=45, twistLimitDegrees=2}
 {bone1Name=dyn_coat_l_05, bone2Name=dyn_coat_l_04, swing1LimitDegrees=10, swing2LimitDegrees=45, twistLimitDegrees=2}
 {
  bone1Name=dyn_coat_bk_l_04
  bone2Name=dyn_coat_l_04
  swing1LimitDegrees=45
  swing1Motion=Free
  swing2LimitDegrees=45
  swing2Motion=Free
  twistLimitDegrees=45
  twistMotion=Free
 }
 {bone1Name=dyn_coat_r_05, bone2Name=dyn_coat_r_04, swing1LimitDegrees=10, swing2LimitDegrees=45, twistLimitDegrees=2}
 {
  bone1Name=dyn_coat_r_04
  bone2Name=dyn_coat_bk_r_04
  swing1LimitDegrees=45
  swing1Motion=Free
  swing2LimitDegrees=45
  swing2Motion=Free
  twistLimitDegrees=45
  twistMotion=Free
 }
]
