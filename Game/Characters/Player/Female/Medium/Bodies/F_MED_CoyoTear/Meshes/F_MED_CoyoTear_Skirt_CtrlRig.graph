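# Control Rig for the CoyoTear female medium body that applies pose-driven corrective offsets to its coat, rope, collar and shoulder pad bones.
# A shared DeformRig function runs first, then a chain of spherical pose readers on the thighs, head and upper arms weights Modify Transforms nodes on those bones.
# Aim constraints point the left and right rope bones at target bones parented to the thighs.
# The two crouch variables are declared but unused in this graph.
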ControlRigBlueprint /Game/Characters/Player/Female/Medium/Bodies/F_MED_CoyoTear/Meshes/F_MED_CoyoTear_Skirt_CtrlRig
graphs [
 RigVMFunctionLibrary {links [], nodes []}
 RigVMModel {nodes [
   RigUnit_BeginExecution = &RigUnit_BeginExecution()
   FN_DeformRig_FNC = RigVMFunctionReferenceNode(
    DeformRig_ Curve="(deformRig_l_upperarm_up_alpha_72_#1=False,deformRig_r_upperarm_up_alpha_73_#2=False,deformRig_l_upperarm_fr_alpha_74_#3=False,deformRig_r_upperarm_fr_alpha_75_CB51B8 ...(truncated)"
    DeformRig_Bool="(upperarm_1_#4=true,elbow_3_#5=true,wrist_7_#6=True,groin_8_#7=true,glute_17_#8=false,knee_9_40C408324CA26 ...(truncated)"
    DeformRig_Debug=(upperarm_debug_18_#4=false,wrist_debug_20_#6=false,lat_debug_26_#9=False,groinGlute_debug_27_#7=false)
    DeformRig_Value="(upperarm_up_18_#10=(Rotation=(X=0.0,Y=-0.0,Z=0.0,W=1.0),Translation=(X=0.0,Y=0.0,Z=8.0),Scale3D=(X=0.70,Y=1.0,Z=1.0)),upperarm_fwd_fr_19_#11=(Rotation=(X=0.0,Y=-0.0,Z=0.0,W=1.0),Translation=(X=0.0, ...(truncated)"
    ExecuteContext=@RigUnit_BeginExecution.ExecutePin
   )
   > n1 = &RigUnit_SphericalPoseReader(
    ActiveRegionScaleFactors=(PositiveWidth=1,NegativeWidth=1.2,PositiveHeight=0.8,NegativeHeight=0.8)
    ActiveRegionSize=0.2
    Debug=(bDrawDebug=false,bDraw2D=false,bDrawLocalAxes=false,DebugScale=25,DebugSegments=20,DebugThickness=0.25)
    DriverAxis=(X=-1,Y=0,Z=0)
    DriverItem=(Type=Bone,Name="thigh_l")
    FalloffRegionScaleFactors=(PositiveWidth=1,NegativeWidth=0.35,PositiveHeight=1,NegativeHeight=1)
    FalloffSize=0.5
    FlipHeightScaling=false
    FlipWidthScaling=false
    OptionalParentItem=(Type=Bone,Name="pelvis")
    RotationOffset=(X=-80,Y=0,Z=-30)
   )
   > &RigUnit_ModifyTransforms(
    ItemToModify=((Item=(Type=Bone,Name="dyn_coat_bk_l_MSTR"),Transform=(Rotation=(X=0,Y=0,Z=0,W=1),Translation=(X=0,Y=0,Z=0),Scale3D=(X=1,Y=1,Z=1))))
    ItemToModify.0.Transform.Rotation=&RigVMFunction_MathQuaternionFromEuler(Euler=(X=0,Y=-30,Z=0)).Result
    Weight=@n1.OutputParam
   )
   > n2 = &RigUnit_SphericalPoseReader(
    ActiveRegionScaleFactors=(PositiveWidth=1,NegativeWidth=1.2,PositiveHeight=0.8,NegativeHeight=0.8)
    ActiveRegionSize=0.2
    Debug=(bDrawDebug=false,bDraw2D=false,bDrawLocalAxes=false,DebugScale=25,DebugSegments=20,DebugThickness=0.25)
    DriverAxis=(X=1,Y=0,Z=0)
    DriverItem=(Type=Bone,Name="thigh_r")
    FalloffRegionScaleFactors=(PositiveWidth=0.35,NegativeWidth=1,PositiveHeight=1,NegativeHeight=1)
    FalloffSize=0.5
    FlipHeightScaling=false
    FlipWidthScaling=false
    OptionalParentItem=(Type=Bone,Name="pelvis")
    RotationOffset=(X=90,Y=0,Z=-35)
   )
   > &RigUnit_ModifyTransforms(
    ItemToModify=((Item=(Type=Bone,Name="dyn_coat_bk_r_MSTR"),Transform=(Rotation=(X=0,Y=0,Z=0,W=1),Translation=(X=0,Y=0,Z=0),Scale3D=(X=1,Y=1,Z=1))))
    ItemToModify.0.Transform.Rotation=&RigVMFunction_MathQuaternionFromEuler(Euler=(X=0,Y=-30,Z=0)).Result
    Weight=@n2.OutputParam
   )
   > &RigUnit_AimConstraintLocalSpaceOffset(
    AdvancedSettings=(DebugSettings=(bEnabled=false,Scale=10,WorldOffset=(Rotation=(X=0,Y=0,Z=0,W=1),Translation=(X=0,Y=0,Z=0),Scale3D=(X=1,Y=1,Z=1))),RotationOrderForFilter=XZY)
    Child=(Type=Bone,Name="dyn_rope_l_MSTR")
    Filter=(bX=True,bY=True,bZ=True)
    Parents=((Item=(Type=Bone,Name="dyn_rope_l_TGT"),Weight=1))
    Weight=1
    WorldUp=(Target=(X=0,Y=0,Z=1),Kind=Direction,Space=(Type=Bone,Name="dyn_rope_l_TGT"))
    bMaintainOffset=false
   )
   > &RigUnit_AimConstraintLocalSpaceOffset(
    AdvancedSettings=(DebugSettings=(bEnabled=false,Scale=10,WorldOffset=(Rotation=(X=0,Y=0,Z=0,W=1),Translation=(X=0,Y=0,Z=0),Scale3D=(X=1,Y=1,Z=1))),RotationOrderForFilter=XZY)
    Child=(Type=Bone,Name="dyn_rope_r_MSTR")
    Filter=(bX=True,bY=True,bZ=True)
    Parents=((Item=(Type=Bone,Name="dyn_rope_r_TGT"),Weight=1))
    Weight=1
    WorldUp=(Target=(X=0,Y=0,Z=1),Kind=Direction,Space=(Type=Bone,Name="dyn_rope_r_TGT"))
    bMaintainOffset=false
   )
   > n3 = &RigUnit_SphericalPoseReader(
    ActiveRegionScaleFactors=(PositiveWidth=1.5,NegativeWidth=1.5,PositiveHeight=1,NegativeHeight=1.5)
    ActiveRegionSize=0.1
    Debug=(bDrawDebug=false,bDraw2D=false,bDrawLocalAxes=true,DebugScale=25,DebugSegments=20,DebugThickness=0.25)
    DriverAxis=(X=-1,Y=0,Z=0)
    DriverItem=(Type=Bone,Name="thigh_l")
    FalloffRegionScaleFactors=(PositiveWidth=1.5,NegativeWidth=1,PositiveHeight=0.6,NegativeHeight=1.2)
    FalloffSize=0.35
    FlipHeightScaling=false
    FlipWidthScaling=true
    OptionalParentItem=(Type=Bone,Name="pelvis")
    RotationOffset=(X=80,Y=0,Z=-40)
   )
   > &RigUnit_ModifyTransforms(
    ItemToModify="((Item=(Type=Bone,Name=\"dyn_rope_MSTR\"),Transform=(Rotation=(X=0,Y=0,Z=0,W=1),Translation=(X=1,Y=0,Z=2),Scale3D=(X=1,Y=1,Z=1))),(Item=(Type=Bone,Name=\"dyn_rope_l_MSTR\"),Transform=(Rotation=(X=0,Y=0,Z=0,W=1),Translation=(X=0,Y=0,Z=6),Scale3D=(X=1,Y=1,Z=1))) ...(truncated)"
    ItemToModify.1.Transform.Rotation=&RigVMFunction_MathQuaternionFromEuler(Euler=(X=0,Y=20,Z=0)).Result
    ItemToModify.0.Transform.Rotation=&RigVMFunction_MathQuaternionFromEuler(Euler=(X=-15,Y=0,Z=0)).Result
    Weight=@n3.OutputParam
   )
   > n4 = &RigUnit_SphericalPoseReader(
    ActiveRegionScaleFactors=(PositiveWidth=0.5,NegativeWidth=0.3,PositiveHeight=1,NegativeHeight=1)
    ActiveRegionSize=0.15
    Debug=(bDrawDebug=false,bDraw2D=false,bDrawLocalAxes=true,DebugScale=25,DebugSegments=20,DebugThickness=0.25)
    DriverAxis=(X=1,Y=0,Z=0)
    DriverItem=(Type=Bone,Name="head")
    FalloffRegionScaleFactors=(PositiveWidth=0.5,NegativeWidth=1,PositiveHeight=0.9,NegativeHeight=1)
    FalloffSize=0.4
    FlipHeightScaling=false
    FlipWidthScaling=true
    OptionalParentItem=(Type=Bone,Name="spine_04")
    RotationOffset=(X=-80,Y=-70,Z=0)
   )
   > &RigUnit_ModifyTransforms(
    ItemToModify=((Item=(Type=Bone,Name="def_collar_bk_01"),Transform=(Rotation=(X=0,Y=0,Z=0,W=1),Translation=(X=-6,Y=0,Z=4),Scale3D=(X=0.5,Y=1,Z=1))))
    WeightMinimum=0
    ItemToModify.0.Transform.Rotation=&RigVMFunction_MathQuaternionFromEuler(Euler=(X=0,Y=-50,Z=0)).Result
    Weight=@n4.OutputParam
   )
   > n5 = &RigUnit_SphericalPoseReader(
    ActiveRegionScaleFactors=(PositiveWidth=1,NegativeWidth=0.8,PositiveHeight=0.7,NegativeHeight=0.3)
    ActiveRegionSize=0.15
    Debug=(bDrawDebug=false,bDraw2D=false,bDrawLocalAxes=true,DebugScale=25,DebugSegments=20,DebugThickness=0.25)
    DriverAxis=(X=1,Y=0,Z=0)
    DriverItem=(Type=Bone,Name="head")
    FalloffRegionScaleFactors=(PositiveWidth=1,NegativeWidth=1.3,PositiveHeight=0.5,NegativeHeight=0.5)
    FalloffSize=0.3
    FlipHeightScaling=false
    FlipWidthScaling=true
    OptionalParentItem=(Type=Bone,Name="spine_05")
    RotationOffset=(X=-45,Y=-15,Z=15)
   )
   > &RigUnit_ModifyTransforms(
    ItemToModify="((Item=(Type=Bone,Name=\"def_collar_bk_r_01\"),Transform=(Rotation=(X=0,Y=0,Z=0,W=1),Translation=(X=-2,Y=0,Z=2),Scale3D=(X=0.6,Y=1,Z=1))),(Item=(Type=Bone,Name=\"def_collar_bk_01\"),Transform=(Rotation=(X=0,Y=0,Z=0,W=1),Translation=(X=0,Y=0,Z=0),Scale3D=(X=1,Y ...(truncated)"
    WeightMinimum=0
    ItemToModify.0.Transform.Rotation=&RigVMFunction_MathQuaternionFromEuler(Euler=(X=0,Y=-60,Z=0)).Result
    ItemToModify.1.Transform.Rotation=&RigVMFunction_MathQuaternionFromEuler(Euler=(X=0,Y=-25,Z=15)).Result
    Weight=@n5.OutputParam
   )
   > n6 = &RigUnit_SphericalPoseReader(
    ActiveRegionScaleFactors=(PositiveWidth=0.4,NegativeWidth=0.6,PositiveHeight=1,NegativeHeight=1)
    ActiveRegionSize=0.15
    Debug=(bDrawDebug=false,bDraw2D=false,bDrawLocalAxes=true,DebugScale=25,DebugSegments=20,DebugThickness=0.25)
    DriverAxis=(X=-1,Y=0,Z=0)
    DriverItem=(Type=Bone,Name="upperarm_r")
    FalloffRegionScaleFactors=(PositiveWidth=0.6,NegativeWidth=0.6,PositiveHeight=0.7,NegativeHeight=1)
    FalloffSize=0.4
    FlipHeightScaling=false
    FlipWidthScaling=true
    OptionalParentItem=(Type=Bone,Name="spine_04")
    RotationOffset=(X=-100,Y=25,Z=40)
   )
   > &RigUnit_ModifyTransforms(
    ItemToModify=((Item=(Type=Bone,Name="def_collar_ft_r_01"),Transform=(Rotation=(X=0,Y=0,Z=0,W=1),Translation=(X=0,Y=5,Z=-5),Scale3D=(X=1,Y=1,Z=1))))
    WeightMinimum=0
    ItemToModify.0.Transform.Rotation=&RigVMFunction_MathQuaternionFromEuler(Euler=(X=-30,Y=20,Z=0)).Result
    Weight=@n6.OutputParam
   )
   > n7 = &RigUnit_SphericalPoseReader(
    ActiveRegionScaleFactors=(PositiveWidth=0.6,NegativeWidth=0.6,PositiveHeight=0.8,NegativeHeight=0.5)
    ActiveRegionSize=0.15
    Debug=(bDrawDebug=false,bDraw2D=false,bDrawLocalAxes=true,DebugScale=25,DebugSegments=20,DebugThickness=0.25)
    DriverAxis=(X=0,Y=1,Z=0)
    DriverItem=(Type=Bone,Name="head")
    FalloffRegionScaleFactors=(PositiveWidth=0.7,NegativeWidth=0.8,PositiveHeight=1,NegativeHeight=0.6)
    FalloffSize=0.3
    FlipHeightScaling=false
    FlipWidthScaling=false
    OptionalParentItem=(Type=Bone,Name="spine_05")
    RotationOffset=(X=140,Y=0,Z=15)
   )
   > &RigUnit_ModifyTransforms(
    ItemToModify="((Item=(Type=Bone,Name=\"def_collar_l_01\"),Transform=(Rotation=(X=0,Y=0,Z=0,W=1),Translation=(X=0,Y=0,Z=2),Scale3D=(X=0.8,Y=1,Z=1))),(Item=(Type=Bone,Name=\"def_collar_bk_l_01\"),Transform=(Rotation=(X=0,Y=0,Z=0,W=1),Translation=(X=-2,Y=0,Z=3),Scale3D=(X=1,Y= ...(truncated)"
    WeightMinimum=0
    ItemToModify.2.Transform.Rotation=&RigVMFunction_MathQuaternionFromEuler(Euler=(X=0,Y=-10,Z=0)).Result
    ItemToModify.1.Transform.Rotation=&RigVMFunction_MathQuaternionFromEuler(Euler=(X=0,Y=-10,Z=0)).Result
    ItemToModify.0.Transform.Rotation=&RigVMFunction_MathQuaternionFromEuler(Euler=(X=0,Y=0,Z=0)).Result
    Weight=@n7.OutputParam
   )
   > n8 = &RigUnit_SphericalPoseReader(
    ActiveRegionScaleFactors=(PositiveWidth=1.8,NegativeWidth=1.5,PositiveHeight=1,NegativeHeight=1)
    ActiveRegionSize=0.1
    Debug=(bDrawDebug=false,bDraw2D=false,bDrawLocalAxes=true,DebugScale=25,DebugSegments=20,DebugThickness=0.25)
    DriverAxis=(X=0,Y=0,Z=1)
    DriverItem=(Type=Bone,Name="upperarm_l")
    FalloffRegionScaleFactors=(PositiveWidth=1,NegativeWidth=1,PositiveHeight=1,NegativeHeight=1)
    FalloffSize=0.3
    FlipHeightScaling=false
    FlipWidthScaling=true
    OptionalParentItem=(Type=Bone,Name="clavicle_l")
    RotationOffset=(X=-90,Y=0,Z=0)
   )
   > &RigUnit_ModifyTransforms(
    ItemToModify=((Item=(Type=Bone,Name="def_should_pad"),Transform=(Rotation=(X=0,Y=0,Z=0,W=1),Translation=(X=2.5,Y=-1,Z=0),Scale3D=(X=1,Y=1,Z=1))))
    WeightMinimum=0
    ItemToModify.0.Transform.Rotation=&RigVMFunction_MathQuaternionFromEuler(Euler=(X=-10,Y=5,Z=8)).Result
    Weight=@n8.OutputParam
   )
   > n9 = &RigUnit_SphericalPoseReader(
    ActiveRegionScaleFactors=(PositiveWidth=1.8,NegativeWidth=1.5,PositiveHeight=1.7,NegativeHeight=0.8)
    ActiveRegionSize=0.1
    Debug=(bDrawDebug=false,bDraw2D=false,bDrawLocalAxes=true,DebugScale=25,DebugSegments=20,DebugThickness=0.25)
    DriverAxis=(X=1,Y=0,Z=0)
    DriverItem=(Type=Bone,Name="thigh_r")
    FalloffRegionScaleFactors=(PositiveWidth=1.5,NegativeWidth=1.5,PositiveHeight=1.2,NegativeHeight=1)
    FalloffSize=0.35
    FlipHeightScaling=false
    FlipWidthScaling=true
    OptionalParentItem=(Type=Bone,Name="pelvis")
    RotationOffset=(X=-80,Y=0,Z=-10)
   )
   > &RigUnit_ModifyTransforms(
    ItemToModify="((Item=(Type=Bone,Name=\"dyn_coat_ft_r_MSTR\"),Transform=(Rotation=(X=0,Y=0,Z=0,W=1),Translation=(X=-3,Y=6,Z=8),Scale3D=(X=1,Y=1,Z=1))),(Item=(Type=Bone,Name=\"dyn_coat_r_MSTR\"),Transform=(Rotation=(X=0,Y=0,Z=0,W=1),Translation=(X=-3,Y=-2,Z=4),Scale3D=(X=1,Y= ...(truncated)"
    ItemToModify.0.Transform.Rotation=&RigVMFunction_MathQuaternionFromEuler(Euler=(X=-10,Y=45,Z=45)).Result
    ItemToModify.1.Transform.Rotation=&RigVMFunction_MathQuaternionFromEuler(Euler=(X=30,Y=20,Z=70)).Result
    ItemToModify.2.Transform.Rotation=&RigVMFunction_MathQuaternionFromEuler(Euler=(X=0,Y=0,Z=-20)).Result
    Weight=@n9.OutputParam
   )
   > n10 = &RigUnit_SphericalPoseReader(
    ActiveRegionScaleFactors=(PositiveWidth=1.2,NegativeWidth=1,PositiveHeight=1.2,NegativeHeight=1.6)
    ActiveRegionSize=0.1
    Debug=(bDrawDebug=false,bDraw2D=false,bDrawLocalAxes=true,DebugScale=25,DebugSegments=20,DebugThickness=0.25)
    DriverAxis=(X=-1,Y=0,Z=0)
    DriverItem=(Type=Bone,Name="thigh_l")
    FalloffRegionScaleFactors=(PositiveWidth=1,NegativeWidth=1,PositiveHeight=1,NegativeHeight=1.2)
    FalloffSize=0.4
    FlipHeightScaling=false
    FlipWidthScaling=true
    OptionalParentItem=(Type=Bone,Name="pelvis")
    RotationOffset=(X=100,Y=10,Z=25)
   )
   > &RigUnit_ModifyTransforms(
    ItemToModify="((Item=(Type=Bone,Name=\"dyn_coat_ft_l_MSTR\"),Transform=(Rotation=(X=0,Y=0,Z=0,W=1),Translation=(X=-4,Y=-5,Z=4),Scale3D=(X=1,Y=1,Z=1))),(Item=(Type=Bone,Name=\"dyn_coat_l_MSTR\"),Transform=(Rotation=(X=0,Y=0,Z=0,W=1),Translation=(X=0,Y=0,Z=3),Scale3D=(X=1,Y=1 ...(truncated)"
    ItemToModify.0.Transform.Rotation=&RigVMFunction_MathQuaternionFromEuler(Euler=(X=20,Y=10,Z=-30)).Result
    ItemToModify.1.Transform.Rotation=&RigVMFunction_MathQuaternionFromEuler(Euler=(X=0,Y=0,Z=-60)).Result
    ItemToModify.2.Transform.Rotation=&RigVMFunction_MathQuaternionFromEuler(Euler=(X=0,Y=0,Z=0)).Result
    Weight=@n10.OutputParam
   )
   > n11 = &RigUnit_SphericalPoseReader(
    ActiveRegionScaleFactors=(PositiveWidth=1.5,NegativeWidth=1.5,PositiveHeight=1.2,NegativeHeight=1.6)
    ActiveRegionSize=0.1
    Debug=(bDrawDebug=false,bDraw2D=false,bDrawLocalAxes=true,DebugScale=25,DebugSegments=20,DebugThickness=0.25)
    DriverAxis=(X=-1,Y=0,Z=0)
    DriverItem=(Type=Bone,Name="thigh_l")
    FalloffRegionScaleFactors=(PositiveWidth=1,NegativeWidth=1.5,PositiveHeight=1.2,NegativeHeight=1.2)
    FalloffSize=0.4
    FlipHeightScaling=false
    FlipWidthScaling=true
    OptionalParentItem=(Type=Bone,Name="pelvis")
    RotationOffset=(X=115,Y=10,Z=-30)
   )
   > &RigUnit_ModifyTransforms(
    ItemToModify="((Item=(Type=Bone,Name=\"dyn_coat_ft_l_MSTR\"),Transform=(Rotation=(X=0,Y=0,Z=0,W=1),Translation=(X=0,Y=-2,Z=3),Scale3D=(X=1,Y=1,Z=1))),(Item=(Type=Bone,Name=\"dyn_coat_l_MSTR\"),Transform=(Rotation=(X=0,Y=0,Z=0,W=1),Translation=(X=-6,Y=0,Z=3),Scale3D=(X=1,Y=1 ...(truncated)"
    ItemToModify.0.Transform.Rotation=&RigVMFunction_MathQuaternionFromEuler(Euler=(X=15,Y=20,Z=-60)).Result
    ItemToModify.1.Transform.Rotation=&RigVMFunction_MathQuaternionFromEuler(Euler=(X=-15,Y=-50,Z=-45)).Result
    ItemToModify.2.Transform.Rotation=&RigVMFunction_MathQuaternionFromEuler(Euler=(X=0,Y=-15,Z=0)).Result
    Weight=@n11.OutputParam
   )
  ]}
]
hierarchy [
 root {loc=<0 0 0>, rot=<0 0 0 1>}
 pelvis {parent=root, loc=<0 -0.5301 98.0571>, rot=<0.0127 0.707 -0.0127 -0.707>}
 spine_01 {parent=pelvis, loc=<6.1208 -0.9328 0>, rot=<0 0 0.034 -0.9994>}
 spine_02 {parent=spine_01, loc=<6.5996 -0.2095 0>, rot=<0 0 0.0238 -0.9997>}
 spine_03 {parent=spine_02, loc=<6.4914 -0.1237 0>, rot=<0 0 0.0149 -0.9999>}
 spine_04 {parent=spine_03, loc=<6.9598 0.0794 0>, rot=<0 0 -0.0132 -0.9999>}
 spine_05 {parent=spine_04, loc=<6.9987 0 0>, rot=<0 0 -0.0015 -1>}
 clavicle_l {parent=spine_05, loc=<5.4831 -0.1053 -3.2946>, rot=<0.0465 -0.7289 -0.009 -0.6829>}
 upperarm_l {parent=clavicle_l, loc=<9.8658 -1.2136 -0.4105>, rot=<0.0164 -0.3245 -0.0269 -0.9454>}
 lowerarm_l {parent=upperarm_l, loc=<26.465 -0.8345 0.259>, rot=<0.0036 0.0161 -0.2298 -0.9731>}
 hand_l {parent=lowerarm_l, loc=<22.7355 -0.3203 -0.6105>, rot=<-0.6784 -0.0223 0.0735 -0.7307>}
 thumb_01_l {parent=hand_l, loc=<1.3118 -0.3191 -2.459>, rot=<0.6546 -0.4042 0.0664 -0.6354>}
 thumb_02_l {parent=thumb_01_l, loc=<5.9336 -0.0448 0.0401>, rot=<-0.0037 0.0213 0.298 -0.9543>}
 thumb_03_l {parent=thumb_02_l, loc=<2.2223 0.3682 -0.0907>, rot=<0.0118 -0.013 0.0066 -0.9998>}
 index_metacarpal_l {parent=hand_l, loc=<1.6535 0.0183 -2.96>, rot=<0.0158 -0.0124 -0.0173 -0.9996>}
 index_01_l {parent=index_metacarpal_l, loc=<8.3008 0.0357 -0.0524>, rot=<-0.0044 0.0485 0.0924 -0.9945>}
 index_02_l {parent=index_01_l, loc=<3.3587 -0.3004 0.0452>, rot=<-0.002 0.0027 0.3225 -0.9466>}
 index_03_l {parent=index_02_l, loc=<2.5578 -0.123 -0.0252>, rot=<-0.002 0.0067 0.2558 -0.9667>}
 middle_metacarpal_l {parent=hand_l, loc=<2.1856 -0.1725 -0.5984>, rot=<0.006 0.0168 -0.0329 -0.9993>}
 middle_01_l {parent=middle_metacarpal_l, loc=<7.3275 -0.2447 0.0129>, rot=<-0.0208 0.0042 0.0858 -0.9961>}
 middle_02_l {parent=middle_01_l, loc=<4.0123 0.1401 0.1175>, rot=<-0.0218 0.0258 0.3512 -0.9357>}
 middle_03_l {parent=middle_02_l, loc=<2.6077 0.0534 0.1373>, rot=<0.0082 0.0188 0.1708 -0.9851>}
 ring_metacarpal_l {parent=hand_l, loc=<2.2074 -0.3625 1.3638>, rot=<-0.0284 0.0722 -0.0458 -0.9959>}
 ring_01_l {parent=ring_metacarpal_l, loc=<7.0295 -0.6197 -0.1446>, rot=<-0.0489 -0.0252 0.1892 -0.9804>}
 ring_02_l {parent=ring_01_l, loc=<3.6174 0.4647 -0.1027>, rot=<0 0 0.2806 -0.9598>}
 ring_03_l {parent=ring_02_l, loc=<2.4152 0.114 0.0199>, rot=<0 0 0.1189 -0.9929>}
 pinky_metacarpal_l {parent=hand_l, loc=<1.7616 -0.9753 3.3814>, rot=<-0.0397 0.0814 -0.069 -0.9935>}
 pinky_01_l {parent=pinky_metacarpal_l, loc=<7.1709 -0.4831 -0.1094>, rot=<-0.0654 -0.0035 0.2242 -0.9723>}
 pinky_02_l {parent=pinky_01_l, loc=<2.8274 0.0893 -0.1043>, rot=<0.0037 0.0011 0.2854 -0.9584>}
 pinky_03_l {parent=pinky_02_l, loc=<2.2046 0.0307 -0.0078>, rot=<-0.0002 0.0098 0.1445 -0.9895>}
 deform_wrist_up_base_l {parent=hand_l, loc=<0 2.3009 0>, rot=<0 0 0 -1>}
 deform_wrist_up_l {parent=deform_wrist_up_base_l, loc=<0 0 0>, rot=<0 0 0 1>}
 lowerarm_twist_02_l {parent=lowerarm_l, loc=<10.6707 0 0>, rot=<0 0 0 1>}
 lowerarm_twist_01_l {parent=lowerarm_l, loc=<18.316 0 0>, rot=<0 0 0 1>}
 deform_elbow_in_base_l {parent=lowerarm_l, loc=<0.1834 3.6062 -0.2572>, rot=<0 0 0 -1>}
 deform_elbow_in_l {parent=deform_elbow_in_base_l, loc=<0 0 0>, rot=<0 0 0 1>}
 upperarm_twist_01_l {parent=upperarm_l, loc=<7.2135 -0.1134 0>, rot=<0 0 0 1>}
 upperarm_twist_02_l {parent=upperarm_l, loc=<19.7998 -0.1134 0>, rot=<0 0 0 1>}
 def_upper_arm_MSTR {parent=upperarm_l, loc=<0.1362 0.3769 1.3198>, rot=<0 0 0.008 -1>}
 def_should_pad {parent=upperarm_l, loc=<3.4776 0.3135 5.1566>, rot=<0 0 0 1>}
 attach_cape {parent=spine_05, loc=<9.3216 -8.4501 0>, rot=<0.0616 -0.7044 -0.0616 -0.7044>}
 attach_backpack {parent=spine_05, loc=<0 -8.35 0>, rot=<-0.0093 -0.707 0.0093 -0.707>}
 neck_01 {parent=spine_05, loc=<11.3138 0.1529 0>, rot=<0 0 -0.2003 -0.9797>}
 neck_02 {parent=neck_01, loc=<4.5627 0 0>, rot=<0 0 0.0504 -0.9987>}
 head {parent=neck_02, loc=<5.3914 0 0>, rot=<0 0 0.115 -0.9934>}
 clavicle_r {parent=spine_05, loc=<5.4831 -0.1053 3.2946>, rot=<0.7289 0.0465 -0.6829 0.009>}
 upperarm_r {parent=clavicle_r, loc=<-9.8658 1.2136 0.4105>, rot=<0.0164 -0.3245 -0.0269 -0.9454>}
 lowerarm_r {parent=upperarm_r, loc=<-26.465 0.8345 -0.259>, rot=<0.0036 0.0161 -0.2298 -0.9731>}
 hand_r {parent=lowerarm_r, loc=<-22.7355 0.3203 0.6106>, rot=<-0.6784 -0.0223 0.0735 -0.7307>}
 thumb_01_r {parent=hand_r, loc=<-1.3118 0.3191 2.459>, rot=<0.6546 -0.4042 0.0664 -0.6354>}
 thumb_02_r {parent=thumb_01_r, loc=<-5.9336 0.0448 -0.0401>, rot=<-0.0037 0.0213 0.298 -0.9543>}
 thumb_03_r {parent=thumb_02_r, loc=<-2.2223 -0.3682 0.0907>, rot=<0.0118 -0.013 0.0066 -0.9998>}
 index_metacarpal_r {parent=hand_r, loc=<-1.6535 -0.0183 2.96>, rot=<0.0158 -0.0124 -0.0173 -0.9996>}
 index_01_r {parent=index_metacarpal_r, loc=<-8.3008 -0.0357 0.0524>, rot=<-0.0044 0.0485 0.0924 -0.9945>}
 index_02_r {parent=index_01_r, loc=<-3.3587 0.3004 -0.0452>, rot=<-0.002 0.0027 0.3225 -0.9466>}
 index_03_r {parent=index_02_r, loc=<-2.5578 0.123 0.0252>, rot=<-0.002 0.0067 0.2558 -0.9667>}
 middle_metacarpal_r {parent=hand_r, loc=<-2.1856 0.1725 0.5984>, rot=<0.006 0.0168 -0.0329 -0.9993>}
 middle_01_r {parent=middle_metacarpal_r, loc=<-7.3275 0.2447 -0.0129>, rot=<-0.0208 0.0042 0.0858 -0.9961>}
 middle_02_r {parent=middle_01_r, loc=<-4.0123 -0.1401 -0.1175>, rot=<-0.0218 0.0258 0.3512 -0.9357>}
 middle_03_r {parent=middle_02_r, loc=<-2.6077 -0.0534 -0.1373>, rot=<0.0082 0.0188 0.1708 -0.9851>}
 ring_metacarpal_r {parent=hand_r, loc=<-2.2074 0.3625 -1.3638>, rot=<-0.0284 0.0722 -0.0458 -0.9959>}
 ring_01_r {parent=ring_metacarpal_r, loc=<-7.0295 0.6197 0.1446>, rot=<-0.0489 -0.0252 0.1892 -0.9804>}
 ring_02_r {parent=ring_01_r, loc=<-3.6174 -0.4647 0.1027>, rot=<0 0 0.2806 -0.9598>}
 ring_03_r {parent=ring_02_r, loc=<-2.4152 -0.114 -0.0199>, rot=<0 0 0.1189 -0.9929>}
 pinky_metacarpal_r {parent=hand_r, loc=<-1.7616 0.9753 -3.3814>, rot=<-0.0397 0.0814 -0.069 -0.9935>}
 pinky_01_r {parent=pinky_metacarpal_r, loc=<-7.1709 0.4831 0.1094>, rot=<-0.0654 -0.0035 0.2242 -0.9723>}
 pinky_02_r {parent=pinky_01_r, loc=<-2.8274 -0.0893 0.1043>, rot=<0.0037 0.0011 0.2854 -0.9584>}
 pinky_03_r {parent=pinky_02_r, loc=<-2.2046 -0.0307 0.0078>, rot=<-0.0002 0.0098 0.1445 -0.9895>}
 lowerarm_twist_02_r {parent=lowerarm_r, loc=<-10.6707 0 0>, rot=<0 0 0 -1>}
 lowerarm_twist_01_r {parent=lowerarm_r, loc=<-18.316 0 0>, rot=<0 0 0 -1>}
 upperarm_twist_01_r {parent=upperarm_r, loc=<-7.2135 0.1134 0>, rot=<0 0 0 -1>}
 deform_upperarm_bk_base_r {parent=upperarm_twist_01_r, loc=<6.4053 4.3746 -0.314>, rot=<0 0 0 -1>}
 deform_upperarm_bk_r {parent=deform_upperarm_bk_base_r, loc=<0 0 0>, rot=<0 0 0 1>}
 deform_upperarm_fr_base_r {parent=upperarm_twist_01_r, loc=<5.6312 -5.092 -0.3354>, rot=<0 0 0 -1>}
 deform_upperarm_fr_r {parent=deform_upperarm_fr_base_r, loc=<0 0 0>, rot=<0 0 0 1>}
 deform_upperarm_up_base_r {parent=upperarm_twist_01_r, loc=<8.891 -0.4825 -3.9517>, rot=<0 0 0 -1>}
 deform_upperarm_up_r {parent=deform_upperarm_up_base_r, loc=<0 0 0>, rot=<0 0 0 1>}
 upperarm_twist_02_r {parent=upperarm_r, loc=<-19.7998 0.1134 0>, rot=<0 0 0 -1>}
 dyn_straps_A_MSTR {parent=spine_05, loc=<7.0556 8.2356 6.7481>, rot=<0.2802 0.6492 0.6885 0.1611>}
 dyn_straps_A_01 {parent=dyn_straps_A_MSTR, loc=<0.5839 0 0>, rot=<-0.0078 -0.0801 0.0806 -0.9935>}
 dyn_straps_A_02 {parent=dyn_straps_A_01, loc=<6.9649 0 0>, rot=<0 -0.1314 0.0157 -0.9912>}
 dyn_straps_A_03 {parent=dyn_straps_A_02, loc=<6.4085 0 0>, rot=<0.0049 -0.1824 0.0106 -0.9831>}
 dyn_straps_B_MSTR {parent=spine_05, loc=<6.4669 8.8924 5.5213>, rot=<0.2802 0.6492 0.6885 0.1611>}
 dyn_straps_B_01 {parent=dyn_straps_B_MSTR, loc=<0.5839 0 0>, rot=<-0.0089 -0.0955 0.084 -0.9918>}
 dyn_straps_B_02 {parent=dyn_straps_B_01, loc=<6.209 0 0>, rot=<0.0004 -0.1163 0.0122 -0.9931>}
 dyn_straps_B_03 {parent=dyn_straps_B_02, loc=<6.4151 0 0>, rot=<0.0049 -0.1821 0.0106 -0.9832>}
 def_collar_bk_r_MSTR {parent=spine_05, loc=<11.0067 -4.1144 7.5105>, rot=<-0.3922 -0.0731 -0.1198 -0.9091>}
 def_collar_bk_r_01 {parent=def_collar_bk_r_MSTR, loc=<0 0 0>, rot=<0 0 0 1>}
 def_collar_bk_MSTR {parent=spine_05, loc=<10.7401 -6.529 0>, rot=<-0.703 -0.0764 -0.0764 -0.703>}
 def_collar_bk_01 {parent=def_collar_bk_MSTR, loc=<0 0 0>, rot=<0 0 0 1>}
 def_collar_bk_l_MSTR {parent=spine_05, loc=<11.0071 -4.1144 -7.5105>, rot=<0.9091 0.1198 0.0731 0.3922>}
 def_collar_bk_l_01 {parent=def_collar_bk_l_MSTR, loc=<0 0 0>, rot=<0 0 0 1>}
 def_collar_ft_r_MSTR {parent=spine_05, loc=<6.4299 5.2828 9.4923>, rot=<0.219 0.0161 -0.2679 -0.9381>}
 def_collar_ft_r_01 {parent=def_collar_ft_r_MSTR, loc=<0 0 0>, rot=<0 0 0 1>}
 def_collar_ft_l_MSTR {parent=spine_05, loc=<6.4303 5.2828 -9.4923>, rot=<0.9381 0.2679 -0.0161 -0.219>}
 def_collar_ft_l_01 {parent=def_collar_ft_l_MSTR, loc=<0 0 0>, rot=<0 0 0 1>}
 def_collar_r_MSTR {parent=spine_05, loc=<9.8486 0.4268 10.1363>, rot=<-0.0035 0.0138 -0.246 -0.9692>}
 def_collar_r_01 {parent=def_collar_r_MSTR, loc=<0 0 0>, rot=<0 0 0 1>}
 def_collar_l_MSTR {parent=spine_05, loc=<9.8485 0.4268 -10.1363>, rot=<0.9692 0.246 -0.0138 0.0035>}
 def_collar_l_01 {parent=def_collar_l_MSTR, loc=<0 0 0>, rot=<0 0 0 1>}
 thigh_l {parent=pelvis, loc=<-1.0855 -0.2109 -8.7053>, rot=<0.0009 0.0355 0.0155 -0.9993>}
 calf_l {parent=thigh_l, loc=<-42.5284 -0.0019 -0.7297>, rot=<0.0006 -0.0161 -0.0391 -0.9991>}
 foot_l {parent=calf_l, loc=<-46.7401 -2.5014 -0.592>, rot=<0.0427 -0.0238 0.0353 -0.9982>}
 ball_l {parent=foot_l, loc=<-3.5658 10.8066 -0.1544>, rot=<0.0001 -0.0001 -0.7063 -0.7079>}
 calf_twist_02_l {parent=calf_l, loc=<-19.1546 -1.7943 -0.2638>, rot=<0.0237 0.0095 -0.0035 -0.9997>}
 calf_twist_01_l {parent=calf_l, loc=<-35.8483 -1.9197 -0.5794>, rot=<0.0237 0.0095 -0.0035 -0.9997>}
 thigh_twist_01_l {parent=thigh_l, loc=<-27.952 0 0>, rot=<0.0013 0 0 -1>}
 dyn_coat_ft_l_MSTR {parent=thigh_twist_01_l, loc=<10.6293 8.3554 0.6516>, rot=<0.0636 -0.8518 -0.5151 -0.0711>}
 dyn_coat_ft_l_01 {parent=dyn_coat_ft_l_MSTR, loc=<0 0 0>, rot=<0 0 0 1>}
 dyn_coat_ft_l_02 {parent=dyn_coat_ft_l_01, loc=<15.291 0 0>, rot=<0 0 0 1>}
 dyn_coat_ft_l_03 {parent=dyn_coat_ft_l_02, loc=<15.291 0 0>, rot=<0 0 0 1>}
 dyn_coat_ft_l_04 {parent=dyn_coat_ft_l_03, loc=<15.291 0 0>, rot=<0 0 0 1>}
 def_coat_l_MSTR {parent=thigh_twist_01_l, loc=<32.2359 -5.1003 -4.5755>, rot=<0.0448 -0.9952 0.0021 -0.0875>}
 dyn_coat_l_MSTR {parent=def_coat_l_MSTR, loc=<9.873 0 0>, rot=<0 0 0 1>}
 dyn_coat_l_01 {parent=dyn_coat_l_MSTR, loc=<0 0 0>, rot=<0 0 0 1>}
 dyn_coat_l_02 {parent=dyn_coat_l_01, loc=<14.2934 0 0>, rot=<0 0 0 1>}
 dyn_coat_l_03 {parent=dyn_coat_l_02, loc=<14.4568 0 0>, rot=<0 0 0 1>}
 dyn_coat_l_04 {parent=dyn_coat_l_03, loc=<14.4568 0 0>, rot=<0 0 0 1>}
 dyn_coat_l_05 {parent=dyn_coat_l_04, loc=<14.4568 0 0>, rot=<0 0 0 1>}
 deform_groin_base_l {parent=thigh_l, loc=<0.4801 5.6776 -0.2486>, rot=<0 0 0 -1>}
 deform_groin_l {parent=deform_groin_base_l, loc=<0 0 0>, rot=<0 0 0 1>}
 dyn_rope_l_TGT {parent=thigh_l, loc=<-40.3685 13.9293 5.1076>, rot=<0 0.7071 0.7071 0>}
 thigh_r {parent=pelvis, loc=<-1.0855 -0.2109 8.7053>, rot=<0.0355 -0.0009 0.9993 0.0155>}
 calf_r {parent=thigh_r, loc=<42.5284 0.0019 0.7297>, rot=<0.0006 -0.0161 -0.0391 -0.9991>}
 foot_r {parent=calf_r, loc=<46.7401 2.5014 0.592>, rot=<0.0427 -0.0238 0.0353 -0.9982>}
 ball_r {parent=foot_r, loc=<3.5658 -10.8066 0.1544>, rot=<0.0001 -0.0001 -0.7063 -0.7079>}
 calf_twist_02_r {parent=calf_r, loc=<19.1546 1.7943 0.2638>, rot=<0.0237 0.0095 -0.0035 -0.9997>}
 calf_twist_01_r {parent=calf_r, loc=<35.8483 1.9197 0.5794>, rot=<0.0237 0.0095 -0.0035 -0.9997>}
 thigh_twist_01_r {parent=thigh_r, loc=<27.952 0 0>, rot=<0.0013 0 0 -1>}
 dyn_coat_ft_r_MSTR {parent=thigh_twist_01_r, loc=<-16.0441 -6.6122 3.7687>, rot=<-0.4091 0.0917 -0.038 -0.9071>}
 dyn_coat_ft_r_01 {parent=dyn_coat_ft_r_MSTR, loc=<0 0 0>, rot=<0 0 0 1>}
 dyn_coat_ft_r_02 {parent=dyn_coat_ft_r_01, loc=<12.1688 0 0>, rot=<0 0 0 1>}
 dyn_grenades_A_MSTR {parent=thigh_twist_01_r, loc=<-28.917 7.8978 3.7205>, rot=<0.1849 0.1799 0.123 -0.9583>}
 dyn_grenades_A_01 {parent=dyn_grenades_A_MSTR, loc=<0 0 0>, rot=<0 0 0 1>}
 dyn_grenades_A_02 {parent=dyn_grenades_A_01, loc=<4.0508 0 3.5159>, rot=<0 0 0 1>}
 dyn_grenades_B_MSTR {parent=thigh_twist_01_r, loc=<-22.4992 8.5163 6.4646>, rot=<0.1173 0.1895 0.1382 -0.965>}
 dyn_grenades_B_01 {parent=dyn_grenades_B_MSTR, loc=<1.7781 0 0.3>, rot=<0 0 0 1>}
 dyn_grenades_B_02 {parent=dyn_grenades_B_01, loc=<4.3245 0 2.1458>, rot=<0 0 0 1>}
 dyn_grenades_A_TGT {parent=thigh_twist_01_r, loc=<-19.8667 6.2065 7.6233>, rot=<0.1849 0.1799 0.123 -0.9583>}
 dyn_grenades_B_TGT {parent=thigh_twist_01_r, loc=<-9.1496 5.1832 12.4378>, rot=<0.1173 0.1895 0.1382 -0.965>}
 def_coat_r_MSTR {parent=thigh_twist_01_r, loc=<-33.5567 2.7373 4.7915>, rot=<0.0026 0.0881 -0.0409 -0.9953>}
 dyn_coat_r_MSTR {parent=def_coat_r_MSTR, loc=<8.0749 0 0>, rot=<0 0 0 1>}
 dyn_coat_r_01 {parent=dyn_coat_r_MSTR, loc=<0 0 0>, rot=<0 0 0 1>}
 dyn_coat_r_02 {parent=dyn_coat_r_01, loc=<15.1862 0 0>, rot=<-0.0008 -0.0339 0 -0.9994>}
 dyn_coat_r_03 {parent=dyn_coat_r_02, loc=<14.2934 0 -0.0001>, rot=<0 -0.0401 0 -0.9992>}
 dyn_coat_r_04 {parent=dyn_coat_r_03, loc=<14.4567 0 0>, rot=<0 0 0 1>}
 dyn_coat_r_05 {parent=dyn_coat_r_04, loc=<14.4569 0 0.0001>, rot=<0 0 0 1>}
 dyn_cloth_r_MSTR {parent=dyn_coat_r_02, loc=<-2.9209 -2.2715 1.0149>, rot=<0.0005 0.0305 -0.0019 -0.9995>}
 dyn_cloth_r_01 {parent=dyn_cloth_r_MSTR, loc=<0 0 0>, rot=<0 0 0 1>}
 deform_groin_base_r {parent=thigh_r, loc=<-0.4801 -5.6776 0.2486>, rot=<0 0 0 -1>}
 deform_groin_r {parent=deform_groin_base_r, loc=<0 0 0>, rot=<0 0 0 1>}
 dyn_rope_r_TGT {parent=thigh_r, loc=<22.6035 -18.0134 4.4437>, rot=<-0.675 0.1937 -0.0522 -0.7101>}
 dyn_rope_MSTR {parent=pelvis, loc=<6.8117 7.1779 -0.5859>, rot=<-0.5089 -0.491 -0.491 -0.5089>}
 dyn_rope_r_MSTR {parent=dyn_rope_MSTR, loc=<4.755 3.0405 -3.1386>, rot=<-0.1326 -0.6183 -0.1624 -0.7575>}
 dyn_rope_r {parent=dyn_rope_r_MSTR, loc=<0 0 0>, rot=<0 0 0 1>}
 dyn_rope_l_MSTR {parent=dyn_rope_MSTR, loc=<6.3116 -2.3514 1.5463>, rot=<0.0249 -0.7084 0.0253 -0.705>}
 dyn_rope_l {parent=dyn_rope_l_MSTR, loc=<0 0 0>, rot=<-0.0043 0.0037 0.3146 -0.9492>}
 dyn_coat_bk_MSTR {parent=pelvis, loc=<5.0939 -9.9441 0>, rot=<0.1078 -0.6988 0.6988 -0.1078>}
 dyn_coat_bk_l_MSTR {parent=dyn_coat_bk_MSTR, loc=<11.0665 7.8366 -0.5844>, rot=<-0.0138 -0.1066 -0.0401 -0.9934>}
 dyn_coat_bk_l_01 {parent=dyn_coat_bk_l_MSTR, loc=<0 0 0>, rot=<0 0 0 1>}
 dyn_coat_bk_l_02 {parent=dyn_coat_bk_l_01, loc=<13.9467 0 0>, rot=<0 0 0 1>}
 dyn_coat_bk_l_03 {parent=dyn_coat_bk_l_02, loc=<13.9467 0 0>, rot=<0 0 0 1>}
 dyn_coat_bk_l_04 {parent=dyn_coat_bk_l_03, loc=<13.9467 0 0>, rot=<0 0 0 1>}
 dyn_coat_bk_l_05 {parent=dyn_coat_bk_l_04, loc=<13.9467 0 0>, rot=<0 0 0 1>}
 dyn_coat_bk_r_MSTR {parent=dyn_coat_bk_MSTR, loc=<11.124 -8.1643 -0.3303>, rot=<0.01 -0.1277 0.0178 -0.9916>}
 dyn_coat_bk_r_01 {parent=dyn_coat_bk_r_MSTR, loc=<0 0 0>, rot=<0 0 0 1>}
 dyn_coat_bk_r_02 {parent=dyn_coat_bk_r_01, loc=<13.9467 0 0.0001>, rot=<0 0 0 1>}
 dyn_coat_bk_r_03 {parent=dyn_coat_bk_r_02, loc=<13.9466 0 0>, rot=<0 0 0 1>}
 dyn_coat_bk_r_04 {parent=dyn_coat_bk_r_03, loc=<13.9467 0 0>, rot=<0 0 0 1>}
 dyn_coat_bk_r_05 {parent=dyn_coat_bk_r_04, loc=<13.9466 0 0>, rot=<0 0 0 1>}
 dyn_ground_collision {parent=root, loc=<0 0 0>, rot=<0 0 0 1>}
]
variables [
 isCrouchSprinting {type=bool, default=False}
 isCrouching {type=bool, default=False}
]
