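# Animation Blueprint for the F_MED_CoyoTear female player body, adding secondary motion to coat, straps, ropes and grenades over the pose copied from the parent mesh.
# A skirt Control Rig base gets additive crouch-idle and sprint bone offsets, while sliding uses a separately posed branch.
# RigidBody, AnimDynamics, weighted constraints and look-ats then drive the dangling parts.
# Each update computes skydive-aware gravity for the coat simulation, and floating disables two cloth dynamics nodes.
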AnimBlueprint /Game/Characters/Player/Female/Medium/Bodies/F_MED_CoyoTear/Meshes/F_MED_CoyoTear_AnimBP
target_skeleton=&/F_MED_CoyoTear_Skeleton
graphs [
 AnimGraph {nodes [
   Blend Poses by bool(BlendPose_1=Component To Local(), BlendTime_0=0.3) {blendType=Linear}
   "Transform (Modify) Bone - Bone: dyn_coat_ft_l_01"(Alpha=1, Translation="0,0,3", Rotation="0,0,0", Scale="1,1,1") {
    translationMode=BMM_Additive
    translationSpace=BCS_BoneSpace
    alphaBoolBlend {blendOption=Linear}
    boneToModify {boneName=dyn_coat_ft_l_01}
    translation {z=3}
   }
   n1 = NOT Boolean(A=Get bIsMoving2D())
   n2 = "Transform (Modify) Bone - Bone: dyn_coat_bk_r_MSTR"!(
    ComponentPose=Local To Component(LocalPose=Use cached pose 'basePose'())
    bAlphaBoolEnabled:bool=@n1
    Translation="0,0,0"
    Rotation="80,-30,-15"
    Scale="1,1,1"
   ) -> (Pose) {
    alphaInputType=Bool
    rotationMode=BMM_Additive
    rotationSpace=BCS_BoneSpace
    scaleMode=BMM_Additive
    scaleSpace=BCS_BoneSpace
    translationMode=BMM_Additive
    translationSpace=BCS_BoneSpace
    alphaBoolBlend {blendOption=Linear, blendOutTime=0.2}
    boneToModify {boneName=dyn_coat_bk_r_MSTR}
    rotation=<rot 80 -30 -15>
   }
   n3 = "Transform (Modify) Bone - Bone: dyn_coat_r_MSTR"!(ComponentPose=@n2, bAlphaBoolEnabled:bool=@n1, Translation="0,0,0", Rotation="60,35,15", Scale="1,1,1") -> (Pose) {
    alphaInputType=Bool
    rotationMode=BMM_Additive
    rotationSpace=BCS_BoneSpace
    scaleMode=BMM_Additive
    scaleSpace=BCS_BoneSpace
    translationMode=BMM_Additive
    translationSpace=BCS_BoneSpace
    alphaBoolBlend {blendOption=Linear, blendOutTime=0.2}
    boneToModify {boneName=dyn_coat_r_MSTR}
    rotation=<rot 60 35 15>
   }
   n4 = "Transform (Modify) Bone - Bone: dyn_coat_ft_r_MSTR"!(ComponentPose=@n3, bAlphaBoolEnabled:bool=@n1, Translation="0,0,0", Rotation="0,30,0", Scale="1,1,1") -> (Pose) {
    alphaInputType=Bool
    rotationMode=BMM_Additive
    rotationSpace=BCS_BoneSpace
    scaleMode=BMM_Additive
    scaleSpace=BCS_BoneSpace
    translationMode=BMM_Additive
    translationSpace=BCS_BoneSpace
    alphaBoolBlend {blendOption=Linear, blendOutTime=0.2}
    boneToModify {boneName=dyn_coat_ft_r_MSTR}
    rotation {yaw=30}
   }
   n5 = "Transform (Modify) Bone - Bone: dyn_coat_bk_l_MSTR"!(ComponentPose=@n4, bAlphaBoolEnabled:bool=@n1, Translation="0,0,0", Rotation="60,0,0", Scale="1,1,1") -> (Pose) {
    alphaInputType=Bool
    rotationMode=BMM_Additive
    rotationSpace=BCS_BoneSpace
    scaleMode=BMM_Additive
    scaleSpace=BCS_BoneSpace
    translationMode=BMM_Additive
    translationSpace=BCS_BoneSpace
    alphaBoolBlend {blendOption=Linear, blendOutTime=0.2}
    boneToModify {boneName=dyn_coat_bk_l_MSTR}
    rotation {pitch=60}
   }
   n6 = "Transform (Modify) Bone - Bone: dyn_coat_ft_l_MSTR"!(ComponentPose=@n5, bAlphaBoolEnabled:bool=@n1, Translation="0,0,0", Rotation="30,0,0", Scale="1,1,1") -> (Pose) {
    alphaInputType=Bool
    rotationMode=BMM_Additive
    rotationSpace=BCS_BoneSpace
    scaleMode=BMM_Additive
    scaleSpace=BCS_BoneSpace
    translationMode=BMM_Additive
    translationSpace=BCS_BoneSpace
    alphaBoolBlend {blendOption=Linear, blendOutTime=0.2}
    boneToModify {boneName=dyn_coat_ft_l_MSTR}
    rotation {pitch=30}
   }
   n7 = "Transform (Modify) Bone - Bone: dyn_coat_l_MSTR"!(ComponentPose=@n6, bAlphaBoolEnabled:bool=@n1, Translation="0,0,0", Rotation="60,0,0", Scale="1,1,1") -> (Pose) {
    alphaInputType=Bool
    rotationMode=BMM_Additive
    rotationSpace=BCS_BoneSpace
    scaleMode=BMM_Additive
    scaleSpace=BCS_BoneSpace
    translationMode=BMM_Additive
    translationSpace=BCS_BoneSpace
    alphaBoolBlend {blendOption=Linear, blendOutTime=0.2}
    boneToModify {boneName=dyn_coat_l_MSTR}
    rotation {pitch=60}
   }
   n8 = Component To Local(ComponentPose=@n7)
   n9 = Use cached pose 'basePose'()
   n10 = Get bIsCrouching()
   n11 = Get bIsSprinting()
   n12 = "Transform (Modify) Bone - Bone: dyn_coat_bk_l_MSTR"(
    ComponentPose=Local To Component(LocalPose=Blend Poses by bool(BlendPose_0=@n8, BlendPose_1=@n9, bActiveValue=@n10) {blendType=Linear})
    bAlphaBoolEnabled=@n11
    Translation="0,0,0"
    Rotation="15,0,0"
    Scale="1,1,1"
   ) {
    alphaInputType=Bool
    rotationMode=BMM_Additive
    rotationSpace=BCS_BoneSpace
    translationMode=BMM_Additive
    translationSpace=BCS_BoneSpace
    alphaBoolBlend {blendInTime=0.2, blendOption=Linear, blendOutTime=0.1}
    boneToModify {boneName=dyn_coat_bk_l_MSTR}
    rotation {pitch=15}
   }
   n13 = "Transform (Modify) Bone - Bone: dyn_coat_bk_r_MSTR"(ComponentPose=@n12, bAlphaBoolEnabled=@n11, Translation="0,0,0", Rotation="15,0,0", Scale="1,1,1") {
    alphaInputType=Bool
    rotationMode=BMM_Additive
    rotationSpace=BCS_BoneSpace
    translationMode=BMM_Additive
    translationSpace=BCS_BoneSpace
    alphaBoolBlend {blendInTime=0.2, blendOption=Linear, blendOutTime=0.1}
    boneToModify {boneName=dyn_coat_bk_r_MSTR}
    rotation {pitch=15}
   }
   n14 = Component To Local(ComponentPose=@n13)
   n15 = "Constraint - dyn_coat_bk_MSTR (P)"(ComponentPose=Local To Component(LocalPose=Copy Pose From Mesh() {bUseAttachedParent=true}), Alpha=1, ConstraintWeights_0=1) {
    alphaBoolBlend {blendOption=Linear}
    boneToModify {boneName=dyn_coat_bk_MSTR}
    constraintSetup [{targetBone {boneName=spine_01}}]
    constraintWeights [1]
   }
   n16 = "Constraint - def_coat_l_MSTR (P)"(ComponentPose=@n15, Alpha=1, ConstraintWeights_0=1) {
    alphaBoolBlend {blendOption=Linear}
    boneToModify {boneName=def_coat_l_MSTR}
    constraintSetup [{targetBone {boneName=spine_01}}]
    constraintWeights [1]
   }
   n17 = "Constraint - def_coat_r_MSTR (P)"(ComponentPose=@n16, Alpha=1, ConstraintWeights_0=1) {
    alphaBoolBlend {blendOption=Linear}
    boneToModify {boneName=def_coat_r_MSTR}
    constraintSetup [{targetBone {boneName=spine_01}}]
    constraintWeights [1]
   }
   n18 = "Transform (Modify) Bone - Bone: dyn_coat_bk_MSTR"(ComponentPose=@n17, Alpha=1, Translation="0,0,-3", Rotation="100,0,0", Scale="1,1,1") {
    rotationMode=BMM_Additive
    rotationSpace=BCS_BoneSpace
    scaleMode=BMM_Additive
    scaleSpace=BCS_BoneSpace
    translationMode=BMM_Additive
    translationSpace=BCS_BoneSpace
    alphaBoolBlend {blendOption=Linear}
    boneToModify {boneName=dyn_coat_bk_MSTR}
    rotation {pitch=100}
    translation {z=-3}
   }
   n19 = "Transform (Modify) Bone - Bone: dyn_coat_bk_r_MSTR"!(ComponentPose=@n18, Alpha:float=1, Translation="0,0,0", Rotation="0,-10,0", Scale="1,1,1") -> (Pose) {
    rotationMode=BMM_Additive
    rotationSpace=BCS_BoneSpace
    scaleMode=BMM_Additive
    scaleSpace=BCS_BoneSpace
    translationMode=BMM_Additive
    translationSpace=BCS_BoneSpace
    alphaBoolBlend {blendOption=Linear}
    boneToModify {boneName=dyn_coat_bk_r_MSTR}
    rotation {yaw=-10}
   }
   n20 = "Transform (Modify) Bone - Bone: def_coat_l_MSTR"(ComponentPose=@n19, Alpha=1, Translation="0,8,0", Rotation="25,-90,0", Scale="1,1,1") {
    rotationMode=BMM_Additive
    rotationSpace=BCS_BoneSpace
    scaleMode=BMM_Additive
    scaleSpace=BCS_BoneSpace
    translationMode=BMM_Additive
    translationSpace=BCS_BoneSpace
    alphaBoolBlend {blendOption=Linear}
    boneToModify {boneName=def_coat_l_MSTR}
    rotation {pitch=25, yaw=-90}
    translation {y=8}
   }
   n21 = "Transform (Modify) Bone - Bone: def_coat_r_MSTR"(ComponentPose=@n20, Alpha=1, Translation="2.069,-2.589,2.785", Rotation="43.289,94.859,21.080", Scale="1,1,1") {
    rotationMode=BMM_Additive
    rotationSpace=BCS_BoneSpace
    scaleMode=BMM_Additive
    scaleSpace=BCS_BoneSpace
    translationMode=BMM_Additive
    translationSpace=BCS_BoneSpace
    alphaBoolBlend {blendOption=Linear}
    boneToModify {boneName=def_coat_r_MSTR}
    rotation=<rot 43.289 94.859 21.08>
    translation=<2.069 -2.5887 2.7853>
   }
   n22 = "Transform (Modify) Bone - Bone: dyn_coat_ft_l_MSTR"!(ComponentPose=@n21, Alpha:float=1, Translation="-5,-10,2", Rotation="0,-100,0", Scale="1,1,1") -> (Pose) {
    rotationMode=BMM_Additive
    rotationSpace=BCS_BoneSpace
    scaleMode=BMM_Additive
    scaleSpace=BCS_BoneSpace
    translationMode=BMM_Additive
    translationSpace=BCS_BoneSpace
    alphaBoolBlend {blendOption=Linear}
    boneToModify {boneName=dyn_coat_ft_l_MSTR}
    rotation {yaw=-100}
    translation=<-5 -10 2>
   }
   n23 = "Transform (Modify) Bone - Bone: dyn_coat_ft_r_MSTR"!(ComponentPose=@n22, Alpha:float=1, Translation="3.352,15.557,6.687", Rotation="-15,80,-10", Scale="1,1,1") -> (Pose) {
    rotationMode=BMM_Additive
    rotationSpace=BCS_BoneSpace
    scaleMode=BMM_Additive
    scaleSpace=BCS_BoneSpace
    translationMode=BMM_Additive
    translationSpace=BCS_BoneSpace
    alphaBoolBlend {blendOption=Linear}
    boneToModify {boneName=dyn_coat_ft_r_MSTR}
    rotation=<rot -15 80 -10>
    translation=<3.352 15.557 6.687>
   }
   n24 = Component To Local(ComponentPose=@n23)
   n25 = Get bIsSliding()
   n26 = RigidBody!(
    ComponentPose=Local To Component(LocalPose=Blend Poses by bool(BlendPose_0=@n24, BlendPose_1=@n14, BlendTime_0=0.35, BlendTime_1=0.35, bActiveValue=@n25) {blendType=Linear})
    Alpha
    OverrideWorldGravity:Vector=Get GravityOverride()
    ExternalForce
   ) -> (Pose) {
    bClampLinearTranslationLimitToRefPose=true
    bOverrideWorldGravity=true
    overridePhysicsAsset=&/F_MED_CoyoTear_Dyn_skirt
    simulationSpace=BaseBoneSpace
    alphaBoolBlend {blendOption=Linear}
    baseBoneRef {boneName=pelvis}
    componentAppliedLinearAccClamp=<3000 3000 3000>
    componentLinearAccScale=<0.4 0.1 0.1>
    componentLinearVelScale=<0.4 0.1 0.1>
    overrideWorldGravity {z=-980}
    simSpaceSettings {
     maxAngularAcceleration=500
     maxAngularVelocity=3
     maxLinearAcceleration=700
     maxLinearVelocity=500
     velocityScaleZ=0
     worldAlpha=1
     externalLinearDragV=<0.1 0.1>
    }
   }
   n27 = NOT Boolean(A=Get bIsSkydiveFloating())
   n28 = Anim Dynamics - Bone: dyn_cloth_r_01!(ComponentPose=@n26, bAlphaBoolEnabled:bool=@n27, ExternalForce="0,0,0") -> (Pose) {
    alphaInputType=Bool
    angularDampingOverride=0.999
    bOverrideAngularDamping=true
    bUseGravityOverride=true
    simulationSpace=RootRelative
    alphaBoolBlend {blendOption=Linear}
    boundBone {boneName=dyn_cloth_r_01}
    componentAppliedLinearAccClamp=<5000 5000 5000>
    componentLinearAccScale=<0.3 0.3 0.1>
    componentLinearVelScale=<0.1 0.1 0.1>
    gravityOverride {z=-980}
    physicsBodyDefinitions [{
      boundBone {boneName=dyn_cloth_r_01}
      boxExtents=<15 20 30>
      constraintSetup {angularLimitsMax=<0 0 0>, angularLimitsMin=<0 -75 0>, angularTarget=<0 0 0>}
      localJointOffset=<10 0 0>
     }]
    relativeSpaceBone {boneName=pelvis}
   }
   n29 = Anim Dynamics - Bone: dyn_coat_ft_r_01!(ComponentPose=@n28, bAlphaBoolEnabled:bool=@n27, ExternalForce="0,0,0") -> (Pose) {
    alphaInputType=Bool
    angularDampingOverride=0.999
    bOverrideAngularDamping=true
    bUseGravityOverride=true
    simulationSpace=RootRelative
    alphaBoolBlend {blendOption=Linear}
    boundBone {boneName=dyn_coat_ft_r_01}
    componentAppliedLinearAccClamp=<3000 3000 3000>
    componentLinearAccScale=<0.3 0.3 0.1>
    componentLinearVelScale=<0.1 0.1 0.1>
    gravityOverride {z=-980}
    physicsBodyDefinitions [{
      boundBone {boneName=dyn_coat_ft_r_01}
      boxExtents=<15 20 30>
      constraintSetup {angularLimitsMax=<0 0 0>, angularLimitsMin=<0 -45 0>, angularTarget=<0 0 0>}
      localJointOffset=<7 0 0>
     }]
    relativeSpaceBone {boneName=pelvis}
   }
   n30 = "Constraint - def_upper_arm_MSTR (PPP)"!(ComponentPose=@n29, Alpha=1, ConstraintWeights_0=0.25, ConstraintWeights_1:float=0.55, ConstraintWeights_2:float=0.2) -> (Pose) {
    alphaBoolBlend {blendOption=Linear}
    boneToModify {boneName=def_upper_arm_MSTR}
    constraintSetup [{targetBone {boneName=spine_05}}, {targetBone {boneName=clavicle_l}}, {targetBone {boneName=upperarm_l}}]
    constraintWeights [1,1,1]
   }
   n31 = Look At - Bone: def_upper_arm_MSTR(ComponentPose=@n30, Alpha=1) {
    bUseLookUpAxis=true
    alphaBoolBlend {blendOption=Linear}
    boneToModify {boneName=def_upper_arm_MSTR}
    lookAtLocation {x=0}
    lookAtTarget {boneReference {boneName=lowerarm_l}}
    lookAt_Axis {axis=<1 0>}
    lookUp_Axis {axis {x=0, z=1}}
   }
   n32 = "Constraint - dyn_straps_A_MSTR (PPPP)"!(
    ComponentPose=@n31
    Alpha=1
    ConstraintWeights_0=0.3892
    ConstraintWeights_1:float=0.1625
    ConstraintWeights_2:float=0.376
    ConstraintWeights_3:float=0.0724
   ) -> (Pose) {
    alphaBoolBlend {blendOption=Linear}
    boneToModify {boneName=dyn_straps_A_MSTR}
    constraintSetup [
     {targetBone {boneName=spine_05}}
     {targetBone {boneName=clavicle_r}}
     {targetBone {boneName=def_collar_ft_r_01}}
     {targetBone {boneName=def_collar_r_01}}
    ]
    constraintWeights [1,1,1,1]
   }
   n33 = "Constraint - dyn_straps_B_MSTR (PPPP)"!(
    ComponentPose=@n32
    Alpha=1
    ConstraintWeights_0=0.557
    ConstraintWeights_1:float=0.1095
    ConstraintWeights_2:float=0.333
    ConstraintWeights_3:float=0.0005
   ) -> (Pose) {
    alphaBoolBlend {blendOption=Linear}
    boneToModify {boneName=dyn_straps_B_MSTR}
    constraintSetup [
     {targetBone {boneName=spine_05}}
     {targetBone {boneName=clavicle_r}}
     {targetBone {boneName=def_collar_ft_r_01}}
     {targetBone {boneName=def_collar_r_01}}
    ]
    constraintWeights [1,1,1,1]
   }
   n34 = RigidBody!(ComponentPose=@n33, bAlphaBoolEnabled:bool, OverrideWorldGravity="0,0,-980", ExternalForce) -> (Pose) {
    alphaInputType=Bool
    bClampLinearTranslationLimitToRefPose=true
    bOverrideWorldGravity=true
    overridePhysicsAsset=&/F_MED_CoyoTear_Dyn_Straps
    simulationSpace=BaseBoneSpace
    alphaBoolBlend {blendOption=Linear}
    baseBoneRef {boneName=spine_05}
    componentAppliedLinearAccClamp=<3000 1000 1000>
    componentLinearAccScale {x=0.15}
    componentLinearVelScale {x=0.2}
    overrideWorldGravity {z=-980}
    simSpaceSettings {maxAngularAcceleration=100, maxAngularVelocity=2, maxLinearAcceleration=500, maxLinearVelocity=100, velocityScaleZ=0, worldAlpha=1}
   }
   n35 = Anim Dynamics - Bone: dyn_rope_l(ComponentPose=@n34, Alpha=1, ExternalForce="0,0,0") {
    angularDampingOverride=0.9999
    bOverrideAngularDamping=true
    bUseGravityOverride=true
    simulationSpace=RootRelative
    alphaBoolBlend {blendOption=Linear}
    boundBone {boneName=dyn_rope_l}
    componentAppliedLinearAccClamp=<3000 3000 3000>
    componentLinearAccScale=<0.3 0.3 0.35>
    componentLinearVelScale=<0.3 0.3 0.3>
    gravityOverride {z=-980}
    physicsBodyDefinitions [{
      boundBone {boneName=dyn_rope_l}
      boxExtents=<10 25 15>
      constraintSetup {angularLimitsMax=<0 5 20>, angularLimitsMin=<0 -60 -20>, angularTarget=<0 0 0>}
      localJointOffset=<5 3.7 0>
     }]
    relativeSpaceBone {boneName=pelvis}
   }
   n36 = Anim Dynamics - Bone: dyn_rope_r(ComponentPose=@n35, Alpha=1, ExternalForce="0,0,0") {
    angularDampingOverride=0.9999
    bOverrideAngularDamping=true
    bUseGravityOverride=true
    simulationSpace=RootRelative
    alphaBoolBlend {blendOption=Linear}
    boundBone {boneName=dyn_rope_r}
    componentAppliedLinearAccClamp=<3000 3000 3000>
    componentLinearAccScale=<0.3 0.3 0.35>
    componentLinearVelScale=<0.3 0.3 0.3>
    gravityOverride {z=-980}
    physicsBodyDefinitions [{
      boundBone {boneName=dyn_rope_r}
      boxExtents=<10 25 15>
      constraintSetup {angularLimitsMax=<0 0 20>, angularLimitsMin=<0 -60 -20>, angularTarget=<0 0 0>}
      localJointOffset=<4 -2 0>
     }]
    relativeSpaceBone {boneName=pelvis}
   }
   n37 = "Constraint - dyn_grenades_A_MSTR (PPPPP)"!(
    ComponentPose=@n36
    Alpha=1
    ConstraintWeights_0=0.4171
    ConstraintWeights_1:float=0.0667
    ConstraintWeights_2:float=0.0135
    ConstraintWeights_3:float=0.3946
    ConstraintWeights_4:float=0.1081
   ) -> (Pose) {
    alphaBoolBlend {blendOption=Linear}
    boneToModify {boneName=dyn_grenades_A_MSTR}
    constraintSetup [
     {targetBone {boneName=pelvis}}
     {targetBone {boneName=dyn_coat_bk_r_01}}
     {targetBone {boneName=dyn_coat_r_01}}
     {targetBone {boneName=dyn_coat_bk_MSTR}}
     {targetBone {boneName=def_coat_r_MSTR}}
    ]
    constraintWeights [1,1,1,1,1]
   }
   n38 = "Constraint - dyn_grenades_B_MSTR (PPPPP)"!(
    ComponentPose=@n37
    Alpha=1
    ConstraintWeights_0=0.1294
    ConstraintWeights_1:float=0.1258
    ConstraintWeights_2:float=0.2435
    ConstraintWeights_3:float=0.3093
    ConstraintWeights_4:float=0.192
   ) -> (Pose) {
    alphaBoolBlend {blendOption=Linear}
    boneToModify {boneName=dyn_grenades_B_MSTR}
    constraintSetup [
     {targetBone {boneName=pelvis}}
     {targetBone {boneName=dyn_coat_bk_r_01}}
     {targetBone {boneName=dyn_coat_r_01}}
     {targetBone {boneName=dyn_coat_bk_MSTR}}
     {targetBone {boneName=def_coat_r_MSTR}}
    ]
    constraintWeights [1,1,1,1,1]
   }
   n39 = "Constraint - dyn_grenades_A_TGT (PPPPPPP)"!(
    ComponentPose=@n38
    Alpha=1
    ConstraintWeights_0=0.0768
    ConstraintWeights_1:float=0.0084
    ConstraintWeights_2:float=0.0005
    ConstraintWeights_3:float=0.5141
    ConstraintWeights_4:float=0.1128
    ConstraintWeights_5:float=0.2322
    ConstraintWeights_6:float=0.0552
   ) -> (Pose) {
    alphaBoolBlend {blendOption=Linear}
    boneToModify {boneName=dyn_grenades_A_TGT}
    constraintSetup [
     {targetBone {boneName=pelvis}}
     {targetBone {boneName=dyn_coat_bk_r_01}}
     {targetBone {boneName=dyn_coat_bk_r_02}}
     {targetBone {boneName=dyn_coat_r_01}}
     {targetBone {boneName=dyn_coat_bk_MSTR}}
     {targetBone {boneName=def_coat_r_MSTR}}
     {targetBone {boneName=dyn_cloth_r_01}}
    ]
    constraintWeights [1,1,1,1,1,1,1]
   }
   n40 = "Constraint - dyn_grenades_B_TGT (PPPPPPPP)"!(
    ComponentPose=@n39
    Alpha=1
    ConstraintWeights_0=0.0086
    ConstraintWeights_1:float=0.0256
    ConstraintWeights_2:float=0.1297
    ConstraintWeights_3:float=0.3703
    ConstraintWeights_4:float=0.0001
    ConstraintWeights_5:float=0.0363
    ConstraintWeights_6:float=0.0269
    ConstraintWeights_7:float=0.4023
   ) -> (Pose) {
    alphaBoolBlend {blendOption=Linear}
    boneToModify {boneName=dyn_grenades_B_TGT}
    constraintSetup [
     {targetBone {boneName=pelvis}}
     {targetBone {boneName=dyn_coat_bk_r_01}}
     {targetBone {boneName=dyn_coat_bk_r_02}}
     {targetBone {boneName=dyn_coat_r_01}}
     {targetBone {boneName=dyn_coat_ft_r_01}}
     {targetBone {boneName=dyn_coat_bk_MSTR}}
     {targetBone {boneName=def_coat_r_MSTR}}
     {targetBone {boneName=dyn_cloth_r_01}}
    ]
    constraintWeights [1,1,1,1,1,1,1,1]
   }
   n41 = Look At - Bone: dyn_grenades_A_01(ComponentPose=@n40, Alpha=1) {
    alphaBoolBlend {blendOption=Linear}
    boneToModify {boneName=dyn_grenades_A_01}
    lookAtLocation {x=0, z=0.5}
    lookAtTarget {boneReference {boneName=dyn_grenades_A_TGT}}
    lookAt_Axis {axis=<1 0>}
    lookUp_Axis {axis=<0 1>}
   }
   n42 = Look At - Bone: dyn_grenades_B_01(ComponentPose=@n41, Alpha=1) {
    alphaBoolBlend {blendOption=Linear}
    boneToModify {boneName=dyn_grenades_B_01}
    lookAtLocation {x=0}
    lookAtTarget {boneReference {boneName=dyn_grenades_B_TGT}}
    lookAt_Axis {axis=<1 0>}
    lookUp_Axis {axis=<0 1>}
   }
   n43 = "Anim Dynamics - Chain: dyn_grenades_A_01 -> dyn_grenades_A_02"(ComponentPose=@n42, Alpha=1, ExternalForce="0,0,0") {
    angularDampingOverride=0.9995
    angularSpringConstant=500
    bChain=true
    bOverrideAngularDamping=true
    bUseGravityOverride=true
    simulationSpace=RootRelative
    alphaBoolBlend {blendOption=Linear}
    boundBone {boneName=dyn_grenades_A_01}
    chainEnd {boneName=dyn_grenades_A_02}
    componentAppliedLinearAccClamp=<3000 3000 3000>
    componentLinearAccScale=<0.1 0.1 0.1>
    componentLinearVelScale=<0.5 0.5 0.5>
    gravityOverride {z=-980}
    physicsBodyDefinitions [
     {
      boundBone {boneName=dyn_grenades_A_01}
      boxExtents=<10 35 20>
      constraintSetup {angularLimitsMax=<0 0 15>, angularLimitsMin=<0 -25 -15>, angularTarget=<0 0 0>}
      localJointOffset=<3 1 0>
     }
     {
      boundBone {boneName=dyn_grenades_A_02}
      boxExtents=<10 30 15>
      constraintSetup {angularLimitsMax=<5 0 15>, angularLimitsMin=<0 -35 -15>, angularTarget=<0 0 0>}
      localJointOffset=<2 2.2 -3>
     }
    ]
    relativeSpaceBone {boneName=pelvis}
   }
   n44 = "Anim Dynamics - Chain: dyn_grenades_B_01 -> dyn_grenades_B_02"(ComponentPose=@n43, Alpha=1, ExternalForce="0,0,0") {
    angularDampingOverride=0.9999
    angularSpringConstant=500
    bChain=true
    bOverrideAngularDamping=true
    bUseGravityOverride=true
    simulationSpace=RootRelative
    alphaBoolBlend {blendOption=Linear}
    boundBone {boneName=dyn_grenades_B_01}
    chainEnd {boneName=dyn_grenades_B_02}
    componentAppliedLinearAccClamp=<3000 3000 3000>
    componentLinearAccScale=<0.1 0.1 0.1>
    componentLinearVelScale=<0.5 0.5 0.5>
    gravityOverride {z=-980}
    physicsBodyDefinitions [
     {
      boundBone {boneName=dyn_grenades_B_01}
      boxExtents=<10 40 15>
      constraintSetup {angularLimitsMax=<0 0 10>, angularLimitsMin=<0 -10 -10>, angularTarget=<0 0 0>}
      localJointOffset=<4 1.8 0>
     }
     {
      boundBone {boneName=dyn_grenades_B_02}
      boxExtents=<10 30 20>
      constraintSetup {angularLimitsMax=<5 0 5>, angularLimitsMin=<0 -10 -5>, angularTarget=<0 0 0>}
      localJointOffset=<2 2.7 -3>
     }
    ]
    relativeSpaceBone {boneName=pelvis}
   }
   n45 = Get bIsSliding()
   n46 = "Transform (Modify) Bone - Bone: dyn_grenades_B_MSTR"!(ComponentPose=@n44, bAlphaBoolEnabled:bool=@n45, Translation="0,0,0", Rotation="20,0,30", Scale="1,1,1") -> (Pose) {
    alphaInputType=Bool
    rotationMode=BMM_Additive
    rotationSpace=BCS_BoneSpace
    scaleMode=BMM_Additive
    scaleSpace=BCS_BoneSpace
    translationMode=BMM_Additive
    translationSpace=BCS_BoneSpace
    alphaBoolBlend {blendInTime=0.3, blendOption=Linear, blendOutTime=0.3}
    boneToModify {boneName=dyn_grenades_B_MSTR}
    rotation {pitch=20, roll=30}
   }
   n47 = "Transform (Modify) Bone - Bone: dyn_grenades_A_01"!(ComponentPose=@n46, bAlphaBoolEnabled:bool=@n45, Translation="0,0,0", Rotation="-10,0,15", Scale="1,1,1") -> (Pose) {
    alphaInputType=Bool
    rotationMode=BMM_Additive
    rotationSpace=BCS_BoneSpace
    scaleMode=BMM_Additive
    scaleSpace=BCS_BoneSpace
    translationMode=BMM_Additive
    translationSpace=BCS_BoneSpace
    alphaBoolBlend {blendInTime=0.3, blendOption=Linear, blendOutTime=0.3}
    boneToModify {boneName=dyn_grenades_A_01}
    rotation {pitch=-10, roll=15}
   }
   Output Pose(Result=Component To Local(ComponentPose=@n47)) {layerGroup=None}
   n48 = Control Rig!(
    Source=Copy Pose From Mesh() {bUseAttachedParent=true}
    Alpha
    isCrouching:bool=Get bIsCrouching()
    isCrouchSprinting:bool=Get bIsCrouchSprinting()
   ) -> (Pose) {
    alphaBoolBlend {blendOption=Linear}
    controlRigAssetReference {blueprintRigClass=&/F_MED_CoyoTear_Skirt_CtrlRig.F_MED_CoyoTear_Skirt_CtrlRig_C}
   }
   Save cached pose 'basePose'(Pose=@n48) {cacheName=basePose}
  ]}
 EventGraph {nodes [
   Event BlueprintUpdateAnimation()
   > n1 = Gravity Override State Function DEPRECATED(CharacterPartAnimInstance=Self-Reference(), GameTimeInSeconds=GetGameTimeInSeconds(), GravityOverrideParams=Get Gravity Override Params())
   > Set GravityOverride(GravityOverride=@n1)
  ]}
]
variables [
 Gravity Override Params {
  type=GravityOverrideParamsStruct
  default {
   skydive_FloatingDown {gravityOverride=<0 -700 400>, windAmplitude=<200 -500 200>, windOffset=<0 -600 200>}
   skydive_GoingUp {gravityOverride=<0 -300 800>, windAmplitude=<200 -500 200>}
  }
 }
 GravityOverride {type=Vector}
 inCrouchingNoMoving {type=boolean}
 thigh_rot_up_l {type=number}
 thigh_rot_up_r {type=number}
 Translation {type=Vector}
]
